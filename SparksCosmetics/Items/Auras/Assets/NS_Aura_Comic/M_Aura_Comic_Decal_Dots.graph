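# Translucent unlit Niagara sprite material for the Comic aura cosmetic, drawing decal-like halftone dots and solid color where the sprite nears scene geometry.
# A Fresnel of flattened scene world normals, stepped into hard bands, splits solid from dotted areas and selects hue-shifted emissive.
# Depth fades and a scribble texture mask opacity, while dynamic parameters control Fresnel, dot scale and inversion.
# Low quality and mobile paths replace the scene-normal Fresnel with a constant vector.
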Material /SparksCosmetics/Items/Auras/Assets/NS_Aura_Comic/M_Aura_Comic_Decal_Dots
graph {
 properties {bUsedWithNiagaraSprites=true, blendMode=BLEND_Translucent, pixelDepthOffsetMode=PDOM_Legacy, shadingModel=MSM_Unlit, usedShadingModels=Unlit}
 nodes [
  n1 = ParticleColor()
  n2 = Multiply(A=Constant() {desc=Emissive Mult, r=5}, B=Multiply(A=@n1.RGB, B=@n1.RGB))
  n3 = MaterialFunctionCall(Hue Shift Percentage=Constant() {desc=Hue Shift, r=0.5}, Texture=@n2) {materialFunction=&HueShift}
  n4 = NamedRerouteUsage()
  n5 = Step(Y=@n4, X=Constant() {r=0.2})
  n6 = OneMinus(None=@n5)
  n7 = QualitySwitch(Default=@n6, Low=Constant(), High=@n6, Medium=@n6, Epic=@n6)
  n8 = Constant()
  MaterialFunctionCall(Emissive=LinearInterpolate(A=@n2, B=Multiply(A=@n3.Result) {constB=5}, Alpha=ShadingPathSwitch(Default=@n7, Mobile=@n8))) {materialFunction=&MF_FX_EmissiveAdjustments}
  n9 = SceneTexture() {sceneTextureId=PPI_WorldNormal}
  n10 = Fresnel(
   ExponentIn=DynamicParameter() {defaultValue=<rgba 3 1 1 1>, paramNames [FresnelExponent, DotInverse, TextureScale, InvertDotFresnel]}.FresnelExponent
   Normal=MaterialFunctionCall(Normal=ComponentMask(None=@n9.Color) {b=true, g=true, r=true}, Flatness=ScalarParameter() {defaultValue=0.15, parameterName=FlattenNormal}) {materialFunction=&FlattenNormal}.Result
  ) {baseReflectFraction=0, exponent=1}
  n11 = VectorParameter() {parameterName=Forward, defaultValue=<rgba 0 0 1 1>}
  NamedRerouteDeclaration(None=ShadingPathSwitch(Default=QualitySwitch(Default=@n10, Low=@n11.RGB, High=@n10, Medium=@n10, Epic=@n10), Mobile=@n11.RGB)) {name=FresnelWorldNormal, nodeColor=<rgba 0 0 0 1>}
  n12 = DepthFade(FadeDistance=ScalarParameter() {defaultValue=15, parameterName=VisibleDepthDistance}) {fadeDistanceDefault=80}
  n13 = DynamicParameter() {defaultValue=<rgba 3 1 1 1>, paramNames [FresnelExponent, DotInverse, TextureScale, InvertDotFresnel]}
  n14 = LinearInterpolate(A=@n5, B=Step(Y=Constant() {r=0.1}, X=@n4), Alpha=@n13.InvertDotFresnel)
  n15 = TextureSample(UVs=MaterialFunctionCall(Texture Scale=@n13.TextureScale) {materialFunction=&ScaleUVsByCenter}.UVs) {texture=&/T_Aura_Comic_Dots}
  n16 = LinearInterpolate(A=@n15.R, B=OneMinus(None=@n15.R), Alpha=@n13.DotInverse)
  n17 = Multiply(
   A=Add(A=@n6, B=Multiply(A=@n14, B=@n16))
   B=DepthFade(Opacity=OneMinus(None=@n12), FadeDistance=ScalarParameter() {defaultValue=15, parameterName=DepthFade}) {fadeDistanceDefault=15}
  )
  Reroute(None=Multiply(A=Multiply(A=@n1.A, B=@n17), B=TextureSample() {texture=&/T_Aura_Comic_Scribble}.G))
 ]
}
