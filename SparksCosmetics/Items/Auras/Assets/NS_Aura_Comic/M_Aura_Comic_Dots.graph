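# Translucent unlit Niagara sprite material for animated comic-style dots in the Comic aura cosmetic.
# Dot texture pans in three stepped jumps per second for a hand-animated look, with dynamic parameters for random offset, scale and inversion.
# Its green channel hue-shifts particle color into emissive adjustments; its blue channel shapes alpha.
# A static switch masks that alpha with a scribble texture (default) or thresholded, step-panned noise, then particle alpha multiplies it.
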Material /SparksCosmetics/Items/Auras/Assets/NS_Aura_Comic/M_Aura_Comic_Dots
graph {
 properties {bUsedWithNiagaraSprites=true, blendMode=BLEND_Translucent, pixelDepthOffsetMode=PDOM_Legacy, shadingModel=MSM_Unlit, usedShadingModels=Unlit}
 nodes [
  n1 = DynamicParameter() {defaultValue=<rgba 1 0 1 1>, paramNames [Randomization, InvertTexture, DotsScale, Param4]}
  n2 = Multiply(A=TextureCoordinate() {uTiling=0.5, vTiling=0.5}, B=@n1.DotsScale)
  n3 = Time()
  n4 = Constant() {r=3}
  n5 = Floor(None=Multiply(A=Frac(None=@n3), B=@n4))
  n6 = TextureSample(UVs=Panner(Coordinate=Add(A=@n2, B=@n1.Randomization) {constB=0}, Time=Divide(A=@n5, B=@n4)) {speedX=1, speedY=1}) {texture=&/T_Aura_Comic_Dots}
  n7 = ParticleColor()
  MaterialFunctionCall(Emissive=MaterialFunctionCall(Hue Shift Percentage=Multiply(A=@n6.G) {constB=0.25}, Texture=@n7.RGB) {materialFunction=&HueShift}.Result) {materialFunction=&MF_FX_EmissiveAdjustments}
  n8 = OneMinus(None=@n6.B)
  n9 = DynamicParameter() {defaultValue=<rgba 1 0 1 1>, paramNames [Randomization, InvertTexture, DotsScale, Param4]}
  n10 = TextureSample() {texture=&/T_Aura_Comic_Scribble}
  n11 = Add(A=TextureCoordinate() {uTiling=0.25, vTiling=0}, B=@n1.Randomization) {constB=0}
  n12 = Frac(None=Time())
  n13 = Constant() {r=3}
  n14 = Divide(A=Floor(None=Multiply(A=@n12, B=@n13)), B=@n13)
  n15 = Step(Y=TextureSample(UVs=Panner(Coordinate=@n11, Time=@n14) {speedX=1, speedY=1}) {samplerType=SAMPLERTYPE_LinearColor, texture=/Game/Global/Textures/Default/Noise/T_Atlas_DigitalNoise_01}.R) {constX=0.5}
  Multiply(
   A=@n7.A
   B=Multiply(
    A=LinearInterpolate(A=@n6.B, B=@n8, Alpha=@n9.InvertTexture)
    B=StaticSwitchParameter(True=@n10.G, False=@n15) {defaultValue=true, parameterName=UseStarMask}
   )
  )
 ]
}
