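# Unlit masked Niagara sprite material that draws the scribble texture for the Comic aura cosmetic.
# Tiling noise modulates the vertical UV gradient, and subtracting particle alpha gives a noisy directional wipe, sharpened by a divide and saturate.
# That wipe multiplies the scribble sample.
# A dynamic "Channel" parameter blends two scribble inputs, but both read the red channel, so it has no visible effect as wired.
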Material /SparksCosmetics/Items/Auras/Assets/NS_Aura_Comic/M_Aura_Comic_Scribble
graph {
 properties {
  bEnableResponsiveAA=true
  bUsedWithNiagaraSprites=true
  blendMode=BLEND_Masked
  pixelDepthOffsetMode=PDOM_Legacy
  shadingModel=MSM_Unlit
  usedShadingModels=Unlit
 }
 nodes [
  n1 = TextureSample() {texture=&/T_Aura_Comic_Scribble}
  n2 = TextureSample() {samplerType=SAMPLERTYPE_LinearColor, texture=/Game/Effects/Fort_Effects/Textures/Tile/LinerNoise64x64}
  n3 = ComponentMask(None=TextureCoordinate()) {g=true}
  n4 = Subtract(A=Add(A=Multiply(A=@n2.R, B=@n3), B=@n3), B=ParticleColor().A)
  Multiply(
   A=LinearInterpolate(A=@n1.R, B=@n1.R, Alpha=DynamicParameter() {paramNames [Channel,Param2,Param3,Param4]}.Channel)
   B=Saturate(None=Divide(A=@n4) {constB=0.1})
  )
  ParticleColor()
 ]
}
